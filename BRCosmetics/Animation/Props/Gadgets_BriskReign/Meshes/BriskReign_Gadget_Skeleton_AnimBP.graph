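# Anim Blueprint for the BriskReign pickaxe gadget prop, adding secondary physics motion to its four-bone tassel.
# Two chains of spring-driven, gravity-affected Anim Dynamics nodes simulate relative to the pickaxe bone, one over the InMotion slot, the other over DefaultSlot, differing only at the first tassel joint.
# bIsMoving2D, not declared here, blends to the InMotion chain when true and DefaultSlot when false; the event graph has no logic.
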AnimBlueprint /BRCosmetics/Animation/Props/Gadgets_BriskReign/Meshes/BriskReign_Gadget_Skeleton_AnimBP
target_skeleton=&/BriskReign_Gadget_Skeleton
graphs [
 AnimGraph {nodes [
   n1 = Anim Dynamics - Bone: JNT_Tassel_1(ComponentPose=Local To Component(LocalPose=Slot 'InMotion'() {slotName=InMotion}), Alpha=1, ExternalForce="0,0,0") {
    angularDampingOverride=0.9999
    angularSpringConstant=100
    bAngularSpring=true
    bOverrideAngularDamping=true
    bUseGravityOverride=true
    bUsePlanarLimit=false
    alphaBoolBlend {blendOption=Linear}
    boundBone {boneName=JNT_Tassel_1}
    componentAppliedLinearAccClamp=<3500 3500 3500>
    componentLinearAccScale=<0.4 0.4 0.4>
    componentLinearVelScale=<0.6 0.6 0.6>
    gravityOverride {z=-980}
    physicsBodyDefinitions [{
      boundBone {boneName=JNT_Tassel_1}
      boxExtents=<12 12 12>
      constraintSetup {angularLimitsMax=<15 120 30>, angularLimitsMin=<-15 -30 -30>, angularTarget=<1 0 0>}
      localJointOffset=<1.5 0 0>
     }]
    relativeSpaceBone {boneName=JNT_Pickaxe}
    simSpaceSettings {maxAngularAcceleration=70, maxAngularVelocity=10, simSpaceAngularAlpha=0.2}
   }
   n2 = Anim Dynamics - Bone: JNT_Tassel_2(ComponentPose=@n1, Alpha=1, ExternalForce="0,0,0") {
    angularDampingOverride=0.9999
    angularSpringConstant=100
    bAngularSpring=true
    bOverrideAngularDamping=true
    bUseGravityOverride=true
    bUsePlanarLimit=false
    alphaBoolBlend {blendOption=Linear}
    boundBone {boneName=JNT_Tassel_2}
    componentAppliedLinearAccClamp=<3500 3500 3500>
    componentLinearAccScale=<0.4 0.4 0.4>
    componentLinearVelScale=<0.6 0.6 0.6>
    gravityOverride {z=-980}
    physicsBodyDefinitions [{
      boundBone {boneName=JNT_Tassel_2}
      boxExtents=<12 12 12>
      constraintSetup {angularLimitsMax=<15 120 30>, angularLimitsMin=<-15 -30 -30>, angularTarget=<1 0 0>}
      localJointOffset=<1.5 0 0>
     }]
    relativeSpaceBone {boneName=JNT_Pickaxe}
    simSpaceSettings {maxAngularAcceleration=70, maxAngularVelocity=10, simSpaceAngularAlpha=0.2}
   }
   n3 = Anim Dynamics - Bone: JNT_Tassel_3(ComponentPose=@n2, Alpha=1, ExternalForce="0,0,0") {
    angularDampingOverride=0.9999
    angularSpringConstant=100
    bAngularSpring=true
    bOverrideAngularDamping=true
    bUseGravityOverride=true
    bUsePlanarLimit=false
    alphaBoolBlend {blendOption=Linear}
    boundBone {boneName=JNT_Tassel_3}
    componentAppliedLinearAccClamp=<3500 3500 3500>
    componentLinearAccScale=<0.4 0.4 0.4>
    componentLinearVelScale=<0.6 0.6 0.6>
    gravityOverride {z=-980}
    physicsBodyDefinitions [{
      boundBone {boneName=JNT_Tassel_3}
      boxExtents=<12 12 12>
      constraintSetup {angularLimitsMax=<15 120 30>, angularLimitsMin=<-15 -30 -30>, angularTarget=<1 0 0>}
      localJointOffset=<1.5 0 0>
     }]
    relativeSpaceBone {boneName=JNT_Pickaxe}
    simSpaceSettings {maxAngularAcceleration=70, maxAngularVelocity=10, simSpaceAngularAlpha=0.2}
   }
   n4 = Anim Dynamics - Bone: JNT_Tassel_4(ComponentPose=@n3, Alpha=1, ExternalForce="0,0,0") {
    angularDampingOverride=0.9999
    angularSpringConstant=100
    bAngularSpring=true
    bOverrideAngularDamping=true
    bUseGravityOverride=true
    bUsePlanarLimit=false
    alphaBoolBlend {blendOption=Linear}
    boundBone {boneName=JNT_Tassel_4}
    componentAppliedLinearAccClamp=<3500 3500 3500>
    componentLinearAccScale=<0.4 0.4 0.4>
    componentLinearVelScale=<0.6 0.6 0.6>
    gravityOverride {z=-980}
    physicsBodyDefinitions [{
      boundBone {boneName=JNT_Tassel_4}
      boxExtents=<12 12 12>
      constraintSetup {angularLimitsMax=<15 120 30>, angularLimitsMin=<-15 -30 -30>, angularTarget=<1 0 0>}
      localJointOffset=<1.5 0 0>
     }]
    relativeSpaceBone {boneName=JNT_Pickaxe}
    simSpaceSettings {maxAngularAcceleration=70, maxAngularVelocity=10, simSpaceAngularAlpha=0.2}
   }
   n5 = Anim Dynamics - Bone: JNT_Tassel_1(ComponentPose=Local To Component(LocalPose=Slot 'DefaultSlot'()), Alpha=1, ExternalForce="0,0,0") {
    angularDampingOverride=0.9999
    angularSpringConstant=100
    bAngularSpring=true
    bOverrideAngularDamping=true
    bUseGravityOverride=true
    bUsePlanarLimit=false
    alphaBoolBlend {blendOption=Linear}
    boundBone {boneName=JNT_Tassel_1}
    componentAppliedLinearAccClamp=<3500 3500 3500>
    componentLinearAccScale=<0.4 0.4 0.4>
    componentLinearVelScale=<0.6 0.6 0.6>
    gravityOverride {z=-980}
    physicsBodyDefinitions [{
      boundBone {boneName=JNT_Tassel_1}
      boxExtents=<12 12 12>
      constraintSetup {angularLimitsMax=<15 180 30>, angularLimitsMin=<-15 -30 -30>, angularTarget=<1 0 0>}
      localJointOffset=<1.0452 0 0.0352>
     }]
    relativeSpaceBone {boneName=JNT_Pickaxe}
    simSpaceSettings {maxAngularAcceleration=70, maxAngularVelocity=10, simSpaceAngularAlpha=0.2}
   }
   n6 = Anim Dynamics - Bone: JNT_Tassel_2(ComponentPose=@n5, Alpha=1, ExternalForce="0,0,0") {
    angularDampingOverride=0.9999
    angularSpringConstant=100
    bAngularSpring=true
    bOverrideAngularDamping=true
    bUseGravityOverride=true
    bUsePlanarLimit=false
    alphaBoolBlend {blendOption=Linear}
    boundBone {boneName=JNT_Tassel_2}
    componentAppliedLinearAccClamp=<3500 3500 3500>
    componentLinearAccScale=<0.4 0.4 0.4>
    componentLinearVelScale=<0.6 0.6 0.6>
    gravityOverride {z=-980}
    physicsBodyDefinitions [{
      boundBone {boneName=JNT_Tassel_2}
      boxExtents=<12 12 12>
      constraintSetup {angularLimitsMax=<15 120 30>, angularLimitsMin=<-15 -30 -30>, angularTarget=<1 0 0>}
      localJointOffset=<1.5 0 0>
     }]
    relativeSpaceBone {boneName=JNT_Pickaxe}
    simSpaceSettings {maxAngularAcceleration=70, maxAngularVelocity=10, simSpaceAngularAlpha=0.2}
   }
   n7 = Anim Dynamics - Bone: JNT_Tassel_3(ComponentPose=@n6, Alpha=1, ExternalForce="0,0,0") {
    angularDampingOverride=0.9999
    angularSpringConstant=100
    bAngularSpring=true
    bOverrideAngularDamping=true
    bUseGravityOverride=true
    bUsePlanarLimit=false
    alphaBoolBlend {blendOption=Linear}
    boundBone {boneName=JNT_Tassel_3}
    componentAppliedLinearAccClamp=<3500 3500 3500>
    componentLinearAccScale=<0.4 0.4 0.4>
    componentLinearVelScale=<0.6 0.6 0.6>
    gravityOverride {z=-980}
    physicsBodyDefinitions [{
      boundBone {boneName=JNT_Tassel_3}
      boxExtents=<12 12 12>
      constraintSetup {angularLimitsMax=<15 120 30>, angularLimitsMin=<-15 -30 -30>, angularTarget=<1 0 0>}
      localJointOffset=<1.5 0 0>
     }]
    relativeSpaceBone {boneName=JNT_Pickaxe}
    simSpaceSettings {maxAngularAcceleration=70, maxAngularVelocity=10, simSpaceAngularAlpha=0.2}
   }
   n8 = Anim Dynamics - Bone: JNT_Tassel_4(ComponentPose=@n7, Alpha=1, ExternalForce="0,0,0") {
    angularDampingOverride=0.9999
    angularSpringConstant=100
    bAngularSpring=true
    bOverrideAngularDamping=true
    bUseGravityOverride=true
    bUsePlanarLimit=false
    alphaBoolBlend {blendOption=Linear}
    boundBone {boneName=JNT_Tassel_4}
    componentAppliedLinearAccClamp=<3500 3500 3500>
    componentLinearAccScale=<0.4 0.4 0.4>
    componentLinearVelScale=<0.6 0.6 0.6>
    gravityOverride {z=-980}
    physicsBodyDefinitions [{
      boundBone {boneName=JNT_Tassel_4}
      boxExtents=<12 12 12>
      constraintSetup {angularLimitsMax=<15 120 30>, angularLimitsMin=<-15 -30 -30>, angularTarget=<1 0 0>}
      localJointOffset=<1.5 0 0>
     }]
    relativeSpaceBone {boneName=JNT_Pickaxe}
    simSpaceSettings {maxAngularAcceleration=70, maxAngularVelocity=10, simSpaceAngularAlpha=0.2}
   }
   Output Pose(Result=Blend Poses by bool(
     BlendPose_0=Component To Local(ComponentPose=@n4)
     BlendPose_1=Component To Local(ComponentPose=@n8)
     BlendTime_0=0.2
     BlendTime_1=0.2
     bActiveValue=Get bIsMoving2D()
    ) {blendType=Linear})
  ]}
 EventGraph {nodes [
   TryGetPawnOwner()
   Event BlueprintUpdateAnimation()
  ]}
]
variables []
